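# Animation Blueprint for the TangyRadishMagma Crust offhand pickaxe that adds physics-driven secondary motion to the axe mesh.
# The cached input pose feeds two RigidBody nodes sharing one physics asset, with master_bow as base bone.
# Each update copies the front-end flag into a bool that blends between them.
# The front end gets partial-strength, zero-gravity world-space simulation; otherwise full-strength base-bone-space simulation with downward gravity applies.
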AnimBlueprint /BRCosmetics/Weapons/FORT_Melee/Pickaxe_TangyRadishMagma_Crust_Offhand/Meshes/TangyRadishMagma_Crust_Offhand_Axe_AnimBP
target_skeleton=&/TangyRadishMagma_Crust_Offhand_Axe_Skeleton
graphs [
 AnimGraph {nodes [
   n1 = RigidBody!(
    ComponentPose=Local To Component(LocalPose=Use cached pose 'input_pose_cache'())
    Alpha=0.25
    OverrideWorldGravity:Vector="0,0,0"
    ExternalForce
   ) -> (Pose) {
    bClampLinearTranslationLimitToRefPose=true
    bOverrideWorldGravity=true
    overridePhysicsAsset=&/TangyRadishMagma_Crust_Offhand_Axe_Physics
    simulationSpace=WorldSpace
    alphaBoolBlend {blendOption=Linear}
    baseBoneRef {boneName=master_bow}
    componentAppliedLinearAccClamp=<0 0 0>
    simSpaceSettings {
     maxAngularAcceleration=1000
     maxAngularVelocity=3
     maxLinearAcceleration=700
     maxLinearVelocity=500
     worldAlpha=1
     externalLinearDragV=<0.2 0.2 0.2>
    }
   }
   n2 = RigidBody(ComponentPose=Local To Component(LocalPose=Use cached pose 'input_pose_cache'()), OverrideWorldGravity="0,0,-980") {
    bClampLinearTranslationLimitToRefPose=true
    bOverrideWorldGravity=true
    overridePhysicsAsset=&TangyRadishMagma_Crust_Offhand_Axe_Physics
    simulationSpace=BaseBoneSpace
    alphaBoolBlend {blendOption=Linear}
    baseBoneRef {boneName=master_bow}
    componentAppliedLinearAccClamp=<0 0 0>
    simSpaceSettings {
     maxAngularAcceleration=1000
     maxAngularVelocity=3
     maxLinearAcceleration=700
     maxLinearVelocity=500
     worldAlpha=1
     externalLinearDragV=<0.2 0.2 0.2>
    }
   }
   Output Pose(Result=Blend Poses by bool(
     BlendPose_0=Component To Local(ComponentPose=@n1)
     BlendPose_1=Component To Local(ComponentPose=@n2)
     bActiveValue=Get use FE float logic() -> (use FE float logic:bool)
    ) {blendType=Linear}) {layerGroup=None}
   AnimGraphNode_SubInput_0 = Input Pose()
   Save cached pose 'input_pose_cache'(Pose=@AnimGraphNode_SubInput_0) {cacheName=input_pose_cache}
  ]}
 EventGraph {nodes [
   Event BlueprintUpdateAnimation()
   > Set use FE float logic(use FE float logic:bool=Get bIsInFrontEnd()) -> (Output_Get:bool)
   TryGetPawnOwner()
  ]}
]
variables [use FE float logic {type=boolean}]
